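# Additive unlit Niagara particle material for wispy, view-dependent sparkles, with particle color passed through an emissive adjustment function.
# Two panning shape textures multiply together, then combine with the tangent-space camera vector to offset a secondary noise lookup.
# That noise's blue channel multiplies the undistorted main noise's, creating sparkles that are heavily brightened.
# The sparkles are shaped by the panning shapes, flame mask and particle alpha, with depth fade skipped on mobile.
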Material /DealershipSharedLibrary/Assets/Sets/Daikon/M_VCVFX_WispySparkles
graph {
 properties {
  bAutomaticallySetUsageInEditor=false
  bUsedWithNiagaraMeshParticles=true
  bUsedWithNiagaraRibbons=true
  bUsedWithNiagaraSprites=true
  blendMode=BLEND_Additive
  shadingModel=MSM_Unlit
  usedShadingModels=Unlit
 }
 nodes [
  MaterialExpressionVertexColor_0 = ParticleColor()
  MaterialFunctionCall(Emissive=@MaterialExpressionVertexColor_0.RGB) {materialFunction=&MF_FX_EmissiveAdjustments}
  n1 = TextureSampleParameter2D() {group=General, parameterName=Mask, texture=/DealershipSharedLibrary/FX/Textures/Gradient_Masks/T_VCVFX_Flame_A}
  n2 = ScalarParameter() {defaultValue=0.6, group=General, parameterName=Mask Brightness}
  n3 = Constant()
  n4 = MaterialFunctionCall!(
   ScaleUVByCenter X
   ScaleUVByCenter Y
   Tiling X=ScalarParameter() {defaultValue=1, group=General, parameterName=Main Textures Tiling X, sortPriority=2}
   Tiling Y=ScalarParameter() {defaultValue=1, group=General, parameterName=Main Textures Tiling Y, sortPriority=3}
   Panner Time=@n3
   Panner Speed X=@n3
   Panner Speed Y=@n3
   Scale U
   Scale V
  ) {materialFunction=/DealershipSharedLibrary/FX/Materials/MaterialFunctions/MF_VCVFX_TilingAndOffset}
  n5 = ComponentMask(None=Transform(None=CameraVectorWS()) {transformSourceType=TRANSFORMSOURCE_World, transformType=TRANSFORM_Tangent}) {g=true, r=true}
  n6 = ScalarParameter() {defaultValue=1, group=Shape1, parameterName=Time Panners}
  n7 = TextureSampleParameter2D(UVs=MaterialFunctionCall(
    Tiling X=ScalarParameter() {defaultValue=1, group=Shape1, parameterName=ShapeTiling X, sortPriority=0}
    Tiling Y=ScalarParameter() {defaultValue=1, group=Shape1, parameterName=ShapeTiling Y, sortPriority=1}
    Panner Time=@n6
    Panner Speed X=ScalarParameter() {defaultValue=0.2, group=Shape1, parameterName=PannerSpeed X, sortPriority=2}
    Panner Speed Y=ScalarParameter() {defaultValue=-0.1, group=Shape1, parameterName=PannerSpeed Y, sortPriority=3}
   ) {materialFunction=&MF_VCVFX_TilingAndOffset}.Result) {group=Shape1, parameterName=Shape TEX, sortPriority=-1, texture=/DealershipSharedLibrary/FX/Textures/Noise/T_VCVFX_Energy_Alpha}
  n8 = TextureSampleParameter2D(UVs=MaterialFunctionCall(
    Tiling X=ScalarParameter() {defaultValue=1, group=Shape2, parameterName=ShapeTiling 2 X, sortPriority=0}
    Tiling Y=ScalarParameter() {defaultValue=1, group=Shape2, parameterName=ShapeTiling 2 Y, sortPriority=1}
    Panner Time=@n6
    Panner Speed X=ScalarParameter() {defaultValue=-0.05, group=Shape2, parameterName=PannerSpeed 2 X, sortPriority=2}
    Panner Speed Y=ScalarParameter() {defaultValue=-0.1, group=Shape2, parameterName=PannerSpeed 2 Y, sortPriority=3}
   ) {materialFunction=&MF_VCVFX_TilingAndOffset}.Result) {group=Shape2, parameterName=Shape 2 TEX, sortPriority=-1, texture=&T_VCVFX_Energy_Alpha}
  n9 = Multiply(A=@n7.RGB, B=@n8.RGB)
  n10 = ComponentMask(None=ConstantBiasScale(None=@n9) {bias=-0.5, scale=2}) {g=true, r=true}
  n11 = TextureSampleParameter2D(UVs=Add(A=@n4.Result, B=Add(A=@n5, B=@n10))) {
   group=General
   parameterName=Secondary Main TEX
   sortPriority=1
   texture=/DealershipSharedLibrary/FX/Textures/Noise/T_VCVFX_NoiseGeneric01Pack_RGB
  }
  n12 = TextureSampleParameter2D(UVs=@n4.Result) {group=General, parameterName=Main TEX, sortPriority=0, texture=&T_VCVFX_NoiseGeneric01Pack_RGB}
  MaterialExpressionConstantClamp_0 = Saturate(None=Multiply(A=@n11.B, B=@n12.B))
  n13 = Multiply(A=@MaterialExpressionConstantClamp_0, B=ScalarParameter() {defaultValue=512, group=General, parameterName=Brightness})
  n14 = Power(Base=@n9, Exp=ScalarParameter() {defaultValue=1, group=Shape1, parameterName="Contrast Shapes (Power)"})
  n15 = Multiply(A=@MaterialExpressionVertexColor_0.A, B=Multiply(A=Multiply(A=@n1.A, B=@n2) {constB=0.6}, B=Multiply(A=@n13, B=@n14)))
  n16 = Multiply(
   A=@n15
   B=DepthFade(FadeDistance=ScalarParameter() {defaultValue=100, group=Opacity, parameterName=DepthFade-Standard_Distance, sortPriority=0}) {fadeDistanceDefault=20}
  )
  ShadingPathSwitch(Default=@n16, Deferred=@n16, Mobile=@n15) {desc=Migrated from FeatureLevelSwitch}
 ]
}
